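; BP_BasketSuburb_CameraFIX is a helper actor that manages the on-screen camera lens effect for the BasketSuburb emote.
; After a short delay on BeginPlay, it casts its attach parent to FortPawn and, only if that pawn is locally viewed, spawns and stores the configured lens effect.
; An OnDestroyed binding casts the stored effect to BP_Emote_BasketSuburb_CameraLens and destroys it, tying the effect's lifetime to this actor.
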(event EventBeginPlay
  (Game|BindEventtoOnDestroyed (delegate Custom|OnDestroyed_Event))
  (Utilities|FlowControl|Delay 0.1)
  (bind _asfort_pawn (Utilities|Casting|CastToFortPawn (Actor|GetAttachParentActor self)))
  (if (Pawn|IsLocallyViewed _asfort_pawn)
    (bind _returnvalue (Feedback|SpawnGenericCameraLensEffect _asfort_pawn (Variables|Components|GetLens%20Effect%20Emitter%20Class)))
    (Variables|Components|SetCameraLensFX _returnvalue)))

(event Custom|OnDestroyed_Event
  (params (DestroyedActor :type Actor%20Object%20Reference))
  (bind _asbp_emote_basket_suburb_camera_lens (Utilities|Casting|CastToBP_Emote_BasketSuburb_CameraLens (Variables|Components|GetCameraLensFX)))
  (Actor|DestroyActor _asbp_emote_basket_suburb_camera_lens))

(fn ConstructionScript)
