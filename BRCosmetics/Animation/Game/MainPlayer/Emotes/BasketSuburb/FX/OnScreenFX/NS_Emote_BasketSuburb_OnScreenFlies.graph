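# Niagara system for the on-screen flies effect of the BasketSuburb emote, built from one burst emitter with two sprite renderers.
# A random walking flag selects curves that shape each fly's velocity, with stop-start motion when walking, and drive two SubUV animations.
# Body tint follows time of day, and both it and the separate wing color darken smoothly when the player is indoors.
# The point attraction force is disabled.
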NiagaraSystem /BRCosmetics/Animation/Game/MainPlayer/Emotes/BasketSuburb/FX/OnScreenFX/NS_Emote_BasketSuburb_OnScreenFlies
dependencies {
 usedDataInterfaces [
  &NiagaraDataInterfaceVectorCurve
  &NiagaraDataInterfaceCurve
  &NiagaraDataInterfaceSpriteRendererInfo
 ]
 usedDynamicInputs []
 usedModules [&SystemState]
}
emitters {DirectionalBurst001 {
  emitterUpdateScript [
   &EmitterState(Life Cycle Mode=::NewEnumerator1, Loop Behavior=::NewEnumerator1, Loop Duration Mode=::NewEnumerator1, UseLoopDelay=false)
   &SpawnBurst_Instantaneous(Use Spawn Probability=false, UseLoopCountLimit=false)
  ]
  particleSpawnScript [
   &InitializeParticle(
    Lifetime=8
    Position Mode=::NewEnumerator1
    Sprite Size=&Multiply_Vector2D_ByFloat(
     Float=0.7
     Vector2D=&MaskVector2dBySpawnGroup(Masked Vector2D=<10 15>, Passthrough Vector2D=<10.4619 9.2942>, Spawn Group Mask=0)
    )
    Sprite Size Mode=::NewEnumerator4
    UsePositionOffset=false
    Write Material Random=true
   )
   &ShapeLocation(
    Handle Radius=1.1887
    MaskSpawnGroup=false
    Override Seed=false
    Rotation Mode=::NewEnumerator1
    Shape Primitive=::NewEnumerator3
    UseOwnerScale=true
   ) {module=ShapeLocation001}
   &AddVelocity(Module.Use Falloff Distance=false, Override Seed=false, Velocity Mode=::NewEnumerator1, Velocity Speed=-100)
   /SetVariables(Particles.WingPos:NiagaraPosition=&AddVectorToPosition(Position=$Particles.Position, Vector=<0 0 0>)) {enabled=false, isSetParameters=true}
  ]
  particleUpdateScript [
   &ParticleState()
   &PointAttractionForce(
    Attraction Radius=193.6169
    Attraction Strength=175.5978
    Falloff Exponent:NiagaraFloat=&RandomRangeFloat(Minimum=0.0727, Maximum=0.2567)
    Kill Within Radius=false
    Module.Use Falloff=true
    UseAttractorOffset=false
   ) {enabled=false}
   /SetVariables(Particles.WalkingBool:NiagaraBool=&RandomBool()) {isSetParameters=true}
   &ScaleVelocity(Velocity Scale=&MakeCustomVectorFromBool(
     Bool=$Particles.WalkingBool
     True Vector=&VectorFromCurve(
      VectorCurve=DataInterface() {propertyValues {
        XCurve {keys [<at 0.0329 1>,<at 1.0329 0.1>]}
        YCurve {keys [
          {interpMode=RCIM_Cubic, leaveTangent=-1.0024, tangentMode=RCTM_Break, time=0.3476, value=0.8199}
          {arriveTangent=6.2519, interpMode=RCIM_Cubic, leaveTangent=6.2519, tangentMode=RCTM_User, time=0.3716, value=0.3782}
          {interpMode=RCIM_Cubic, leaveTangent=-2.1553, tangentMode=RCTM_Break, time=0.7101, value=0.4257}
         ]}
        ZCurve {keys [
          {arriveTangent=0.521, interpMode=RCIM_Cubic, leaveTangent=0.521, tangentMode=RCTM_User, time=0.3534, value=0.5616}
          {interpMode=RCIM_Cubic, leaveTangent=5.2894, tangentMode=RCTM_Break, time=0.364, value=1.097}
          {arriveTangent=-0.0093, interpMode=RCIM_Cubic, leaveTangent=-0.0093, tangentMode=RCTM_User, time=0.7417, value=1.0482}
         ]}
       }}
      Scale Curve=&VectorFromFloat(Value=&FloatFromCurve(FloatCurve=DataInterface() {propertyValues {Curve {keys [<at 0.1198 1 RCIM_Constant>,<at 0.1765 0.001 RCIM_Constant>,<at 0.3563 1.0118 RCIM_Constant>,<at 0.5603 0.0128 RCIM_Constant>,<at 0.7367 1 RCIM_Constant>,<at 1.0183 0.001>]}}}))
     )
     False Vector=&VectorFromCurve(
      VectorCurve=DataInterface() {propertyValues {
        XCurve {keys [<at 0 0.9857>, <at 0 1>, {interpMode=RCIM_Cubic, tangentMode=RCTM_User, time=0.5175, value=1.3114}, <at 0.9962 1.1016>]}
        YCurve {keys [
          {interpMode=RCIM_Cubic, tangentMode=RCTM_User, time=0, value=1}
          {arriveTangent=1.2392, interpMode=RCIM_Cubic, tangentMode=RCTM_Break, time=0.1436, value=1.4261}
          {interpMode=RCIM_Cubic, tangentMode=RCTM_Break, time=0.4167, value=0.6592}
          {arriveTangent=-2.3621, interpMode=RCIM_Cubic, leaveTangent=-0.0711, tangentMode=RCTM_Break, time=0.8924, value=1.2543}
          {arriveTangent=-1.204, tangentMode=RCTM_User, time=1.0076, value=1.2911}
         ]}
        ZCurve {keys [
          {interpMode=RCIM_Cubic, tangentMode=RCTM_Break, time=0.151, value=0.7549}
          {arriveTangent=-0.2972, interpMode=RCIM_Cubic, leaveTangent=-0.2972, tangentMode=RCTM_User, time=0.4301, value=1.4702}
          {arriveTangent=-2.5703, interpMode=RCIM_Cubic, leaveTangent=0.0508, tangentMode=RCTM_Break, time=0.8835, value=0.8315}
          {arriveTangent=-4.1018, tangentMode=RCTM_User, time=1.0484, value=0.3321}
         ]}
       }}
      Scale Curve=&VectorFromFloat(Value=&FloatFromCurve(FloatCurve=DataInterface() {propertyValues {Curve {keys [<at 0.1198 1 RCIM_Constant>,<at 0.2206 1.0118 RCIM_Constant>,<at 0.6339 1 RCIM_Constant>]}}}))
     )
    ))
   &SolveForcesAndVelocity(Clamp Velocity=false, Limit Acceleration=false)
   &SubUVAnimation(
    Output Alternate Renderer Binding=true
    "Sprite Renderer "=DataInterface() {propertyValues {SpriteRenderer=&self:DirectionalBurst001_1.NiagaraSpriteRendererProperties_1}}
    Start Frame Offset=1
    SubUV Lookup Index:NiagaraFloat=&MakeCustomFloatFromBool(
     Bool=$Particles.WalkingBool
     True Float=&FloatFromCurve(FloatCurve=DataInterface() {propertyValues {Curve {keys [<at 0.1198 0 RCIM_Constant>,<at 0.1869 0.5 RCIM_Constant>,<at 0.2033 0 RCIM_Constant>,<at 0.2465 0.5 RCIM_Constant>,<at 0.2631 0 RCIM_Constant>,<at 0.5741 0.5 RCIM_Constant>,<at 0.5921 0 RCIM_Constant>,<at 0.706 0.5 RCIM_Constant>,<at 0.7241 0 RCIM_Constant>,<at 1.0183 0.001>]}}})
     False Float=&FloatFromCurve(FloatCurve=DataInterface() {propertyValues {Curve {keys [<at 0.1198 0 RCIM_Constant>,<at 0.1869 0.5 RCIM_Constant>,<at 0.2033 0 RCIM_Constant>,<at 0.4168 0.5 RCIM_Constant>,<at 0.4367 0 RCIM_Constant>,<at 0.6278 0.5 RCIM_Constant>,<at 0.6482 0 RCIM_Constant>,<at 1.0183 0.001>]}}})
    )
    UseEndFrame=false
    UseStartFrame=false
   ) {module=SubUVAnimation001}
   &SubUVAnimation(
    "Sprite Renderer "=DataInterface() {propertyValues {SpriteRenderer=&self:DirectionalBurst001_1.NiagaraSpriteRendererProperties_0}}
    Start Frame Offset:NiagaraInt32=&RandomRangeInt(Minimum=0, Maximum=5)
    SubUV Lookup Index Scale:NiagaraFloat=&MakeCustomFloatFromBool(
     Bool=$Particles.WalkingBool
     True Float=&FloatFromCurve(
      FloatCurve=DataInterface() {propertyValues {Curve {keys [<at 0.1198 1 RCIM_Constant>,<at 0.1765 0 RCIM_Constant>,<at 0.3563 1.0118 RCIM_Constant>,<at 0.5603 0 RCIM_Constant>,<at 0.7367 1 RCIM_Constant>,<at 1.0183 0.001>]}}}
      Scale Curve=17
     )
     False Float=&FloatFromCurve(
      FloatCurve=DataInterface() {propertyValues {Curve {keys [<at 0.1198 1 RCIM_Constant>,<at 0.2206 1.0118 RCIM_Constant>,<at 0.6339 1 RCIM_Constant>]}}}
      Scale Curve=17
     )
    )
    UseEndFrame=false
    UseStartFrame=false
   )
   &InterpolateOverTime(
    Rate Of Change=5
    Target Float Value:NiagaraFloat=&RemapRange(Input Value To Remap=$NPC.FortniteNPC.IsPlayerIndoors, Input Min=0, Input Max=0.2, Output Min=0, Output Max=1)
   )
   &ScaleColor(
    Scale RGB=&Lerp_Vector(
     A=&MakeVectorFromLinearColorRGB(Color=&FNTimeOfDayColor(
       Night Color=<rgba 0.7865 0.1716 0.5456 1>
       Morning Color=<rgba 0.5729 0 0 1>
       Day Color=<rgba 0.5729 0 0 1>
       Evening Color=<rgba 0.5729 0 0 1>
      ))
     B=&MakeVectorFromLinearColorRGB(Color=&FNTimeOfDayColor(Night Color=<rgba 0.1354 0.0295 0.0939 1>, Morning Color=<rgba 0.099 0 0 1>, Day Color=<rgba 0.099 0 0 1>, Evening Color=<rgba 0.099 0 0 1>))
     Alpha=$Particles.InterpolateOverTime.InterpolateOverTime_Float.Moving Average
    )
    ScaleA=true
    ScaleRGB=true
   )
   /SetVariables(
    Particles.WingColor:LinearColor=&LerpLinearColors(
     StartColor=<rgba 1 1 1 1>
     EndColor=<rgba 0.2135 0.2135 0.2135 1>
     LerpFactor=$Particles.InterpolateOverTime.InterpolateOverTime_Float.Moving Average
    )
    Particles.WingSize:Vector2f=&Multiply_Vector2D_ByFloat(Float=0.7, Vector2D=<7.2565 9.7872>)
   ) {isSetParameters=true}
  ]
  rendererClasses [&NiagaraSpriteRendererProperties]
  renderers [
   {rendererClass=&NiagaraSpriteRendererProperties}
   {rendererClass=&NiagaraSpriteRendererProperties, rendererIndex=1}
  ]
 }}
summary {
 systemName=NS_Emote_BasketSuburb_OnScreenFlies
 emitters [{emitterName=DirectionalBurst001, rendererClasses [&NiagaraSpriteRendererProperties]}]
 userVariables []
}
